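# Sound cue that plays the radio tower activation start sound for the mid-match radio tower objective.
# A single wave player outputs RadioTower_Active_Start with no randomization or mixing.
# It uses the radio tower dishes attenuation asset and routes to the in-game SFX sound class.
# Concurrency overrides specify a single instance that stops the oldest one.
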SoundCue /MidMatchRadioTowerGameplay/Sound/MMObj_RadioTower_Activate_Cue
root=WavePlayer_4
nodes [{id=WavePlayer_4, properties {soundWaveAssetPtr=&/RadioTower_Active_Start}}]
properties {
 attenuationSettings=&/MMObj_RadioTowerDishes_Att
 maxDistance=25000
 priority=0.5
 soundClassObject=&Fort_SFX_InGame
 volumeMultiplier=1
 concurrencyOverrides {maxCount=1, resolutionRule=StopOldest}
 concurrencySet [null]
}
