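# Looping sound cue for the mid-match radio tower's active state, playing the distant activation loop wave.
# A single looping wave player feeds the root, with distant attenuation that carries it far across the map and a boosted volume.
# Concurrency limits it to one instance with PreventNew resolution, and bus sends are disabled under the in-game SFX class.
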SoundCue /MidMatchRadioTowerGameplay/Sound/MMObj_RadioTower_Activate_Loop_Cue
root=WavePlayer_3
nodes [{id=WavePlayer_3, properties {bLooping=true, soundWaveAssetPtr=&/RadioTower_Active_Distnt_Loop}}]
properties {
 attenuationSettings=&/MMObj_RadioTower_Distant_Att
 bEnableBusSends=false
 maxDistance=18000
 soundClassObject=&Fort_SFX_InGame
 volumeMultiplier=2
 concurrencyOverrides {maxCount=1, resolutionRule=PreventNew}
 concurrencySet [&/RadioTowerPower_Loop_Conc]
}
