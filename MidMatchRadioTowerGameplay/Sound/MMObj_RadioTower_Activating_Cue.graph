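# Sound cue for the mid-match radio tower objective, playing the tower's activation start sound.
# A single wave player feeds the RadioTower_Active_Start wave straight to the output.
# It uses the radio tower dishes attenuation asset with long-range audibility and the in-game SFX sound class.
# The radio tower console concurrency set is assigned, alongside override settings that allow one instance and block new ones.
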SoundCue /MidMatchRadioTowerGameplay/Sound/MMObj_RadioTower_Activating_Cue
root=WavePlayer_1
nodes [{id=WavePlayer_1, properties {soundWaveAssetPtr=&/RadioTower_Active_Start}}]
properties {
 attenuationSettings=&/MMObj_RadioTowerDishes_Att
 maxDistance=25000
 soundClassObject=&Fort_SFX_InGame
 volumeMultiplier=1
 concurrencyOverrides {maxCount=1, resolutionRule=PreventNew}
 concurrencySet [&/RadioTowerConsole_Conc]
}
