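# Sound cue for the boss spawn in the mid-match radio tower objective, using the boss fight sound class and AI spawn attenuation.
# A mixer layers a vault door opening, enveloped to fade out, with a random pick of four NPC materialize sounds under slight pitch modulation.
# The mixer's first input is empty.
# A concurrency override allows one instance and blocks new ones, and virtualization is disabled.
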SoundCue /MidMatchRadioTowerGameplay/Sound/MMObj_RadioTower_Boss_Spawn_Cue
root=Mixer_0
nodes [
 {
  id=Enveloper_0
  children [WavePlayer_0]
  properties {pitchCurve {editorCurveData {keys [{time=0}]}}, volumeCurve {editorCurveData {keys [<at 0 1>,<at 2 0>]}}}
 }
 {id=Mixer_0, children [null,Enveloper_0,Modulator_0], properties {inputVolume [1,0.6,0.7]}}
 {id=Modulator_0, children [Random_0], properties {pitchMax=1.03, pitchMin=1}}
 {id=Random_0, children [WavePlayer_2,WavePlayer_3,WavePlayer_4,WavePlayer_5], properties {weights [1,1,1,1]}}
 {id=WavePlayer_0, properties {soundWaveAssetPtr=/FNGameplayCues/Sounds/Fort_Doors/World/Vault/Vault_Door_Open_01}}
 {id=WavePlayer_2, properties {soundWaveAssetPtr=/Game/Athena/Sounds/NPCs/NPC_Spawn_Materialize_RezIn_04}}
 {id=WavePlayer_3, properties {soundWaveAssetPtr=/Game/Athena/Sounds/NPCs/NPC_Spawn_Materialize_RezIn_03}}
 {id=WavePlayer_4, properties {soundWaveAssetPtr=/Game/Athena/Sounds/NPCs/NPC_Spawn_Materialize_RezIn_02}}
 {id=WavePlayer_5, properties {soundWaveAssetPtr=/Game/Athena/Sounds/NPCs/NPC_Spawn_Materialize_RezIn_01}}
]
properties {
 attenuationSettings=/Lager/Sounds/Settings/Lager_AI_Spawn_Att
 bOverrideConcurrency=true
 maxDistance=6200
 soundClassObject=&Fort_BossFight
 virtualizationMode=Disabled
 volumeMultiplier=1.4
 concurrencyOverrides {maxCount=1, resolutionRule=PreventNew}
}
