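# Sound cue for activating a radio tower console in the mid-match radio tower gameplay, in the in-game SFX sound class.
# A single wave player feeds RadioTowerConsole_Activate_Update straight to the output.
# Playback uses the close-range radio console attenuation, with bus sends disabled.
# A concurrency override allows one instance at a time and prevents new ones from starting while it plays.
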SoundCue /MidMatchRadioTowerGameplay/Sound/RadioTowerConsole_Activate_Cue
root=WavePlayer_0
nodes [{id=WavePlayer_0, properties {soundWaveAssetPtr=&/RadioTowerConsole_Activate_Update}}]
properties {
 attenuationSettings=&/RadioConsole_Close_Att
 bEnableBusSends=false
 bOverrideConcurrency=true
 maxDistance=4000
 soundClassObject=&Fort_SFX_InGame
 volumeMultiplier=1
 concurrencyOverrides {maxCount=1, resolutionRule=PreventNew}
}
