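# Looping sound cue for the mid-match radio tower console's activate/update state, positioned with close exterior attenuation.
# A mixer layers two looping waves, with the console loop dominant and the tower loop much quieter.
# An Enveloper at the root fades the mix from near silence to full volume over about two seconds along an accelerating curve.
# Concurrency settings reference a console loop set, with overrides specifying a single instance that prevents new ones.
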SoundCue /MidMatchRadioTowerGameplay/Sound/RadioTowerConsole_Activate_Loop_Update_Cue
root=Enveloper_0
nodes [
 {
  id=Enveloper_0
  children [Mixer_0]
  properties {
   pitchCurve {editorCurveData {keys [{time=0}]}}
   volumeCurve {editorCurveData {keys [<at 0 0.01>,<at 0.3125 0.0741>,<at 1.0258 0.2024>,<at 1.5541 0.4562>,<at 1.9743 1>]}}
  }
 }
 {id=Mixer_0, children [WavePlayer_0,WavePlayer_1], properties {inputVolume [1.25,0.12]}}
 {id=WavePlayer_0, properties {bLooping=true, soundWaveAssetPtr=&/RadioTowerConsole_Active_Update_Loop}}
 {id=WavePlayer_1, properties {bLooping=true, soundWaveAssetPtr=&/RadioTower_Active_Update_Loop}}
]
properties {
 attenuationSettings=&/RadioTower_Exterior_Close_Att
 bEnableBusSends=false
 maxDistance=4500
 soundClassObject=&Fort_SFX_InGame
 volumeMultiplier=1
 concurrencyOverrides {maxCount=1, resolutionRule=PreventNew}
 concurrencySet [&/RadioTowerConsole_Loop_Conc]
}
