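# Sound cue for the radio tower console activation sound in the mid-match radio tower gameplay.
# A single wave player feeds the RadioTowerConsole_Activate wave straight to the output with a volume boost, using close-range console attenuation and the in-game SFX sound class.
# Overridden concurrency allows only one instance at a time and blocks new plays while one is active, with bus sends disabled.
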SoundCue /MidMatchRadioTowerGameplay/Sound/RadioTowerConsole_Activate_Update_Cue
root=WavePlayer_1
nodes [{id=WavePlayer_1, properties {soundWaveAssetPtr=&/RadioTowerConsole_Activate}}]
properties {
 attenuationSettings=&/RadioConsole_Close_Att
 bEnableBusSends=false
 bOverrideConcurrency=true
 maxDistance=4000
 soundClassObject=&Fort_SFX_InGame
 volumeMultiplier=1.4
 concurrencyOverrides {maxCount=1, resolutionRule=PreventNew}
}
