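# Sound cue for a radio tower console being captured in the mid-match radio tower gameplay.
# A mixer layers three waves: a keycard insert at once, the console captured sound after a short delay, and a slightly pitched-down capture alert a couple of seconds later.
# It plays in the in-game SFX sound class with the radio tower dishes attenuation asset and a boosted overall volume.
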SoundCue /MidMatchRadioTowerGameplay/Sound/RadioTowerConsole_Captured_Cue
root=Mixer_0
nodes [
 {id=Delay_0, children [WavePlayer_1], properties {delayMax=0.2, delayMin=0.2}}
 {id=Delay_1, children [Modulator_0], properties {delayMax=2, delayMin=2}}
 {id=Mixer_0, children [WavePlayer_0,Delay_0,Delay_1], properties {inputVolume [1,1,1]}}
 {id=Modulator_0, children [WavePlayer_2], properties {pitchMax=0.9, pitchMin=0.9}}
 {id=WavePlayer_0, properties {soundWaveAssetPtr=&/RadioTower_KeyCard_Insert}}
 {id=WavePlayer_1, properties {soundWaveAssetPtr=&/RadioTowerConsole_Captured}}
 {id=WavePlayer_2, properties {soundWaveAssetPtr=&/s26_RadioTower_Captured_Alert}}
]
properties {
 attenuationSettings=&/MMObj_RadioTowerDishes_Att
 bEnableBusSends=false
 maxDistance=25000
 soundClassObject=&Fort_SFX_InGame
 volumeMultiplier=1.4
}
