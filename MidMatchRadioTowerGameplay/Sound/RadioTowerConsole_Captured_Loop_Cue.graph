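# Sound cue for the looping audio of a captured radio tower console in mid-match gameplay.
# An Enveloper fades a looping wave player of the console's active loop up from near silence over about two seconds.
# The cue halves pitch and volume, reuses an exterior ambient attenuation, and references a loop concurrency asset alongside a single-instance override that prevents new plays.
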SoundCue /MidMatchRadioTowerGameplay/Sound/RadioTowerConsole_Captured_Loop_Cue
root=Enveloper_0
nodes [
 {
  id=Enveloper_0
  children [WavePlayer_2]
  properties {
   pitchCurve {editorCurveData {keys [{time=0}]}}
   volumeCurve {editorCurveData {keys [<at 0 0.01>,<at 0.3125 0.0741>,<at 1.0258 0.2024>,<at 1.5541 0.4562>,<at 1.9743 1>]}}
  }
 }
 {id=WavePlayer_2, properties {bLooping=true, soundWaveAssetPtr=&/RadioTowerConsole_Active_Loop}}
]
properties {
 attenuationSettings=/Game/Sounds/Apollo/Biomes/Interior_Banks/AmbientAudio_Exterior_Att
 bEnableBusSends=false
 maxDistance=3000
 pitchMultiplier=0.5
 soundClassObject=&Fort_SFX_InGame
 volumeMultiplier=0.5
 concurrencyOverrides {maxCount=1, resolutionRule=PreventNew}
 concurrencySet [&/RadioTowerConsole_Loop_Conc]
}
