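# Sound cue for the tick sound of the radio tower console in mid-match radio tower gameplay.
# A single wave player feeds a modulator that adds slight random pitch variation to each play.
# It plays at reduced volume through close-range console attenuation on the in-game SFX class, with bus sends disabled.
# Its own concurrency override allows a few simultaneous instances, stopping the oldest first.
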SoundCue /MidMatchRadioTowerGameplay/Sound/RadioTowerConsole_Tick_Cue
root=Modulator_0
nodes [
 {id=Modulator_0, children [WavePlayer_0], properties {pitchMax=1.01, pitchMin=0.99}}
 {id=WavePlayer_0, properties {soundWaveAssetPtr=&/RadioTower_Console_Tick}}
]
properties {
 attenuationSettings=&/RadioConsole_Close_Att
 bEnableBusSends=false
 bOverrideConcurrency=true
 maxDistance=4000
 soundClassObject=&Fort_SFX_InGame
 volumeMultiplier=0.5
 concurrencyOverrides {maxCount=3, resolutionRule=StopOldest}
}
