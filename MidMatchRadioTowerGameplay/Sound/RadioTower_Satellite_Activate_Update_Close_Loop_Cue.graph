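# Sound Cue for the close-range loop of the radio tower satellite's active/update sound, in the in-game SFX class.
# A single looping Wave Player feeds the output directly with a boosted volume and bus sends disabled.
# Inline attenuation overrides take precedence over the assigned RadioConsole_Close_Att asset.
# Overridden concurrency allows only one instance at a time and blocks new ones while it plays.
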SoundCue /MidMatchRadioTowerGameplay/Sound/RadioTower_Satellite_Activate_Update_Close_Loop_Cue
root=WavePlayer_1
nodes [{id=WavePlayer_1, properties {bLooping=true, soundWaveAssetPtr=&/RadioTower_Active_Update_Close_Loop}}]
properties {
 attenuationSettings=&/RadioConsole_Close_Att
 bEnableBusSends=false
 bOverrideAttenuation=true
 bOverrideConcurrency=true
 maxDistance=8000
 soundClassObject=&Fort_SFX_InGame
 volumeMultiplier=1.5
 attenuationOverrides {falloffDistance=5000, attenuationShapeExtents {x=3000}}
 concurrencyOverrides {maxCount=1, resolutionRule=PreventNew}
}
